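# Niagara effect for the LanternStroll emote, spawning bats and ghosts around the emote.
# Bats are mesh and sprite particles moved by scripted curve and waveform offsets from the emitter rather than forces, with WispyTrail spawning fading green ribbons from them.
# Ghosts spawn in delayed windows on a ring, split across three mesh renderers by partition-driven visibility tags, with curve-animated material parameters, time-of-day tint and separate gameplay and front-end offsets.
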NiagaraSystem /BRCosmetics/Animation/Game/MainPlayer/Emotes/LanternStroll/FX/NS_Emote_LanternStroll_Ghosts
dependencies {
 usedDataInterfaces [
  &NiagaraDataInterfaceMeshRendererInfo
  &NiagaraDataInterfaceCurve
  &NiagaraDataInterfaceColorCurve
  &NiagaraDataInterfaceParticleRead
  &NiagaraDataInterfaceSkeletalMesh
 ]
 usedDynamicInputs []
 usedModules [
  &self:SystemSpawnScript.SystemScriptSource.SystemScriptGraph.NiagaraNodeAssignment_0.SetVariables_#1
  &SystemState
 ]
}
emitters {
 Bats {
  emitterUpdateScript [
   &EmitterState(Life Cycle Mode=::NewEnumerator1, Loop Behavior=::NewEnumerator1, Loop Duration=0.933, UseLoopDelay=false)
   &SpawnRate(SpawnRate=8, Use Spawn Probability=true)
  ]
  particleSpawnScript [
   &InitializeParticle(
    Color Maximum=<rgba 25 0.4252 0 1>
    Color Minimum=<rgba 0.0123 0.0605 0.0729 1>
    Color Mode=::NewEnumerator2
    Lifetime Max=1.2
    Lifetime Min=0.8
    Lifetime Mode=::NewEnumerator1
    Mesh Renderer Array Visibility Mode=::NewEnumerator3
    Mesh Renderer Info=DataInterface() {propertyValues {MeshRenderer=&self:Bats_7.NiagaraMeshRendererProperties_0}}
    Mesh Scale Mode=::NewEnumerator3
    Mesh Uniform Scale=&Multiply_Float(
     A=&RandomRangeFloat(Minimum=1, Maximum=1.4)
     B=&MakeCustomFloatFromBool(Bool=&SetBoolByFloatComparison(A=$Emitter.Age, B=2), True Float=0, False Float=1)
    )
    Position Offset=<0 200 200>
    Sprite Rotation Mode=::NewEnumerator2
    Sprite Size Mode=::NewEnumerator1
    Uniform Sprite Size Max=80
    Uniform Sprite Size Min=50
    UsePositionOffset=true
    Write Material Random=true
   )
   &ShapeLocation(MaskSpawnGroup=false, Non Uniform Scale=<3 1 1>, Override Seed=false, Sphere Radius=12, UseOwnerScale=true)
   &AddVelocity(Velocity=&RandomRangeVector(Minimum=<-90 -425 -120>, Maximum=<90 -600 -120>)) {enabled=false}
   /SetVariables(Particles.GlowCO=-10) {isSetParameters=true}
   &InitialMeshOrientation(Mesh Axis To Orient=<0 0 -1>, Mesh Orientation Mode=::NewEnumerator2, Use Rotation Vector=true)
  ]
  particleUpdateScript [
   &ParticleState()
   /SetVariables(Particles.Position=&AddVectorToPosition(
     Position=$Engine.Emitter.SimulationPosition
     Vector=&MakeVector(
      X=&Add_Float(
       A=&FloatFromCurve(
        FloatCurve=DataInterface() {propertyValues {Curve {keys [<at 0 0 RCIM_Cubic>,<at 1 1 RCIM_Cubic>]}}}
        Scale Curve=&RandomRangeFloat(Minimum=-40, Maximum=40)
       )
       B=&Waveform(
        "Waveform [1]"=::NewEnumerator2
        "[1] Frequency"=&RandomRangeFloat(Minimum=0.2, Maximum=0.7)
        "[1] Amplitude Min/Max"=&RandomRangeVector2D(Minimum=<-25 25>, Maximum=<-40 40>)
       )
      )
      Y=&Add_Float(
       A=200
       B=&FloatFromCurve(
        FloatCurve=DataInterface() {propertyValues {Curve {keys [<at 0 0 RCIM_Cubic>,<at 1 -1 RCIM_Cubic>]}}}
        Scale Curve=&RandomRangeFloat(Minimum=325, Maximum=400)
       )
      )
      Z=&Add_Float(
       A=200
       B=&FloatFromCurve(
        FloatCurve=DataInterface() {propertyValues {Curve {keys [<at 0.075 0 RCIM_Cubic>,<at 0.4 -1 RCIM_Cubic>,<at 1 0 RCIM_Cubic>]}}}
        Scale Curve=&RandomRangeFloat(Minimum=25, Maximum=40)
       )
      )
     )
    )) {isSetParameters=true}
   &UpdateMeshOrientation(
    Facing Interpolation=false
    Orientation Method=::NewEnumerator2
    Override Quat=false
    TransformFacingOrigin=false
    TransformFacingPosition=false
    TransformUp=false
   )
   &VortexForce(Origin Pull Amount=500, UseOriginOffset=false, Vortex Force Amount=500) {enabled=false}
   &Drag(Drag=&RandomRangeFloat(Minimum=0.8, Maximum=1.2), Use Linear Drag=true, Use Rotational Drag=true) {enabled=false}
   &CurlNoiseForce(
    Bias Noise=true
    Bias Noise Field:Vector3f=&RandomRangeVector(Minimum=<-1 0 0>, Maximum=<1 0 0>)
    Coordinate Space:ENiagaraCoordinateSpace=::Local
    Noise Frequency=150
    Noise Strength=&RandomRangeFloat(Minimum=200, Maximum=2000)
    Pan Noise=true
    Pan Noise Field=&RandomRangeVector(Minimum=<-1 0 0>, Maximum=<1 0 0>)
   ) {enabled=false}
   &GravityForce(
    Coordinate Space=::Local
    Gravity=&MakeVector(
     X=0
     Y=0
     Z=&FloatFromCurve(
      FloatCurve=DataInterface() {propertyValues {Curve {keys [<at 0.2 1 RCIM_Cubic>, {arriveTangent=-10.8, interpMode=RCIM_Cubic, leaveTangent=-10.8, time=0.25, value=0}, <at 0.45 -1.7 RCIM_Cubic>]}}}
      Scale Curve=-175
     )
    )
   ) {enabled=false}
   &SolveForcesAndVelocity(Clamp Velocity=false, Limit Acceleration=false)
   &ScaleMeshSize(Scale Factor=&VectorFromFloat(Value=&FloatFromCurve(FloatCurve=DataInterface() {propertyValues {Curve {keys [
          {arriveTangent=9.0725, interpMode=RCIM_Cubic, leaveTangent=9.0725, tangentMode=RCTM_User, time=0, value=0}
          <at 0.1408 0.9847 RCIM_Cubic>
          {arriveTangent=-4.2304, interpMode=RCIM_Cubic, leaveTangent=-4.2304, tangentMode=RCTM_User, time=1, value=0}
         ]}}})))
   &ScaleColor(
    Scale Alpha=&FloatFromCurve(FloatCurve=DataInterface() {propertyValues {Curve {keys [
         <at 0 0 RCIM_Cubic>
         {arriveTangent=1.25, leaveTangent=1.25, time=0.2, value=1}
         {interpMode=RCIM_Cubic, tangentMode=RCTM_User, time=0.8, value=1}
         {interpMode=RCIM_Cubic, leaveTangent=-16.5933, time=1, value=0}
        ]}}})
    Scale RGB=&Multiply_VectorByFloat(Vector=<1 1 1>, Float=$System.TOD)
    ScaleA=true
    ScaleRGB=true
   )
   &DynamicMaterialParameters(
    Index 0 Param 1=&RandomRangeFloat(Minimum=0, Maximum=1)
    Index 0 Param 2=&RandomRangeFloat(Minimum=0.9, Maximum=1.25)
    Module.Index0Param0Write=true
    Module.Index0Param1Write=true
    Module.Index0Param2Write=true
    Module.Index0Param3Write=true
   )
   /SetVariables(Particles.GlowColor=&MultiplyLinearColor(
     Linear Color=&RandomRangeLinearColor(Minimum=<rgba 0.222 0.5 0.0129 1>, Maximum=<rgba 0.0114 0.9 0.1797 1>)
     Scale Factor=&MakeVector4_FromLinearColor(LinearColor=&ColorFromCurve(ColorCurve=DataInterface() {propertyValues {
         AlphaCurve {keys [<at 0 0>,<at 0.3 0.35 RCIM_Cubic>,<at 0.85 0 RCIM_Cubic>]}
         BlueCurve {keys [<at 0 1 RCIM_Cubic>,<at 1 1 RCIM_Cubic>]}
         GreenCurve {keys [<at 0 1 RCIM_Cubic>,<at 1 1 RCIM_Cubic>]}
         RedCurve {keys [<at 0 1 RCIM_Cubic>,<at 1 1 RCIM_Cubic>]}
        }}))
    )) {isSetParameters=true}
  ]
  rendererClasses [&NiagaraMeshRendererProperties, &NiagaraSpriteRendererProperties]
  renderers [
   {rendererClass=&NiagaraMeshRendererProperties, rendererIndex=0}
   {rendererClass=&NiagaraSpriteRendererProperties, rendererIndex=1}
  ]
 }
 Ghosts {
  emitterUpdateScript [
   &EmitterState(Life Cycle Mode=::NewEnumerator1, Loop Behavior=::NewEnumerator1, Loop Delay=1.4333, Loop Duration=4.0167, UseLoopDelay=true)
   &SpawnRate(
    SpawnRate=&FloatFromCurve(FloatCurve=DataInterface() {propertyValues {Curve {keys [<at 0.25 1>,<at 0.35 0>]}}}, CurveIndex=$Emitter.NormalizedLoopAge, Scale Curve=5)
    Use Spawn Probability=false
   )
  ]
  particleSpawnScript [
   &InitializeParticle(
    Color=&MultiplyLinearColor(
     Linear Color=<rgba 0.1533 0.8177 0.6655 0.6348>
     Scale Factor=&MakeVector4_FromLinearColor(LinearColor=&FNTimeOfDayColor(
       Night Color=<rgba 0.25 0.55 1.25 1>
       Morning Color=<rgba 1 0.85 0.7 1>
       Day Color=<rgba 0.5 0.6667 1 1>
       Evening Color=<rgba 0.8 0.5667 0.4 1>
      ))
    )
    Lifetime Max=1.8
    Lifetime Min=1.2
    Lifetime Mode=::NewEnumerator1
    Mesh Scale Mode=::NewEnumerator1
    Mesh Uniform Scale Max=1.4
    Position Mode=::NewEnumerator0
    Write Material Random=true
   )
   &PartitionParticles(Number Of Partitions=3, Partition Mode:ENiagara_PartitionMode_v2=::NewEnumerator5)
   &ShapeLocation(
    Large Radius=40
    MaskSpawnGroup=false
    Offset=&FN_FrontEndVector3Modifier(Gameplay=<0 250 55>, Front End=<0 100 55>)
    Offset Mode=::NewEnumerator0
    Override Seed=false
    Rotation Angle=0.25
    Rotation Axis=<0 1 0>
    Rotation Mode=::NewEnumerator1
    Shape Primitive=::NewEnumerator3
    UseOwnerScale=true
   )
   &AddVelocity(
    Module.Use Falloff Distance=false
    Origin Offset=&FN_FrontEndVector3Modifier(Gameplay=<0 500 -8>, Front End=<0 250 -8>)
    Override Seed=false
    Velocity Mode=::NewEnumerator1
    Velocity Speed=&RandomRangeFloat(Minimum=10, Maximum=200)
   )
   &AddVelocity(
    Cone Angle=25
    Cone Axis=<0 -1 0.25>
    Module.UseConeFalloff=false
    Module.UseVelDistribution=false
    Override Seed=false
    Velocity Mode=::NewEnumerator2
    Velocity Speed=&FN_FrontEndFloatModifier(Gameplay=0, Front End=&RandomRangeFloat(Minimum=10, Maximum=100))
   ) {enabled=false, module=AddVelocity001}
  ]
  particleUpdateScript [
   &ParticleState()
   /SetVariables(Particles.VisibilityTag=$Particles.PartitionParticles.Partition) {isSetParameters=true}
   &ScaleMeshSize(Scale Factor=&VectorFromFloat(Value=&FloatFromCurve(FloatCurve=DataInterface() {propertyValues {Curve {
         defaultValue=1
         keys [
          {arriveTangent=2.0553, interpMode=RCIM_Cubic, leaveTangent=2.0553, tangentMode=RCTM_User, time=0, value=0.6}
          {arriveTangent=0.0309, interpMode=RCIM_Cubic, leaveTangent=0.0309, tangentMode=RCTM_User, time=0.5, value=0.9847}
         ]
        }}})))
   &DynamicMaterialParameters(
    Index 0 Param 1=&FloatFromCurve(FloatCurve=DataInterface() {propertyValues {Curve {keys [<at 0 -2.5>,<at 1 1>]}}})
    Index 0 Param 2=&FloatFromCurve(
     FloatCurve=DataInterface() {propertyValues {Curve {keys [{arriveTangent=-0.0039, time=0.5, value=0.1}]}}}
     Scale Curve=&RandomRangeFloat(Minimum=0.25, Maximum=0.4)
    )
    Index 0 Param 3=&FloatFromCurve(FloatCurve=DataInterface() {propertyValues {Curve {keys [
         {arriveTangent=-1.8961, interpMode=RCIM_Cubic, leaveTangent=-1.8961, tangentMode=RCTM_User, time=0, value=0.3}
         {arriveTangent=-0.0708, interpMode=RCIM_Cubic, leaveTangent=-0.0708, tangentMode=RCTM_User, time=0.2, value=0}
         {arriveTangent=0.0092, interpMode=RCIM_Cubic, leaveTangent=0.0092, tangentMode=RCTM_User, time=0.525, value=0}
        ]}}})
    Index 0 Param 4=&RandomRangeFloat(Minimum=0, Maximum=1)
    Index 1 Param 1=&FloatFromCurve(
     FloatCurve=DataInterface() {propertyValues {Curve {keys [
         {arriveTangent=0.0825, interpMode=RCIM_Cubic, leaveTangent=0.0825, tangentMode=RCTM_User, time=0.5, value=0}
         {arriveTangent=2.7761, interpMode=RCIM_Cubic, leaveTangent=2.7761, tangentMode=RCTM_User, time=1, value=1}
        ]}}}
     Scale Curve=&MergeHelper_FloatSwitch(
      0=&RandomRangeFloat(Minimum=0.8, Maximum=1)
      1=&RandomRangeFloat(Minimum=0.8, Maximum=1)
      2=-0.8
      3=?
      Index=&SpawnGroup()
      Num=::NewEnumerator1
     )
    )
    Index 1 Param 2=&FloatFromCurve(
     FloatCurve=DataInterface() {propertyValues {Curve {keys [
         {arriveTangent=-1.6252, interpMode=RCIM_Cubic, time=0, value=0.5}
         {arriveTangent=1.1111, interpMode=RCIM_Cubic, leaveTangent=1.1111, time=0.3, value=1}
         {arriveTangent=0.3571, interpMode=RCIM_Cubic, leaveTangent=0.3571, time=0.45, value=1}
         {interpMode=RCIM_Cubic, leaveTangent=0.1446, time=1, value=1.25}
        ]}}}
     Scale Curve=&MergeHelper_FloatSwitch(0=0.85, 1=0.65, 2=0.8, 3=?, Index=&SpawnGroup(), Num=::NewEnumerator1)
    )
    Index 2 Param 1=&FloatFromCurve(FloatCurve=DataInterface() {propertyValues {Curve {keys [
         {arriveTangent=-0.5375, time=0, value=0}
         {arriveTangent=0.328, interpMode=RCIM_Cubic, leaveTangent=0.328, tangentMode=RCTM_User, time=0.4, value=0}
         {arriveTangent=0.2119, interpMode=RCIM_Cubic, leaveTangent=0.2119, tangentMode=RCTM_User, time=1, value=1}
        ]}}})
    Index 2 Param 2=&MakeFloatFromVector(Vector=<0 0 0>, Channel=::NewEnumerator1)
    Index 2 Param 3=&FloatFromCurve(
     FloatCurve=DataInterface() {propertyValues {Curve {keys [
         {interpMode=RCIM_Cubic, tangentMode=RCTM_User, time=0, value=1}
         {arriveTangent=-0.0336, interpMode=RCIM_Cubic, leaveTangent=-0.0336, tangentMode=RCTM_User, time=0.5, value=1}
         {interpMode=RCIM_Cubic, tangentMode=RCTM_User, time=1, value=0}
        ]}}}
     Scale Curve=&RandomRangeFloat(Minimum=0.8, Maximum=1.1)
    )
    Index 2 Param 4=&FloatFromCurve(FloatCurve=DataInterface() {propertyValues {Curve {keys [
         {interpMode=RCIM_Cubic, tangentMode=RCTM_User, time=0, value=1}
         {arriveTangent=0.6667, interpMode=RCIM_Cubic, leaveTangent=0.6667, time=0.2, value=0.55}
         <at 0.3 1.2 RCIM_Cubic>
         {arriveTangent=-0.0265, interpMode=RCIM_Cubic, leaveTangent=-0.0265, tangentMode=RCTM_User, time=0.4, value=0.55}
         {arriveTangent=0.0518, interpMode=RCIM_Cubic, leaveTangent=0.0518, tangentMode=RCTM_User, time=0.5, value=0.9}
         {interpMode=RCIM_Cubic, tangentMode=RCTM_User, time=0.8, value=0}
        ]}}})
    Module.Index0Param0Write=true
    Module.Index0Param1Write=true
    Module.Index0Param2Write=true
    Module.Index0Param3Write=true
    Module.Index1Param0Write=false
    Module.Index1Param1Write=true
    Module.Index1Param2Write=true
    Module.Index1Param3Write=true
    Module.Index2Param0Write=false
    Module.Index2Param1Write=false
    Module.Index2Param2Write=true
    Module.Index2Param3Write=true
    Write Parameter Index 1=true
    Write Parameter Index 2=true
   )
   &ScaleColor(
    Scale Alpha=&FloatFromCurve(FloatCurve=DataInterface() {propertyValues {Curve {keys [<at 0 0>,<at 0.35 1>,<at 0.7 1>,<at 1 0>]}}}, Scale Curve=0.6)
    Scale RGB=&MakeVectorFromLinearColorRGB(Color=&ColorFromCurve(ColorCurve=DataInterface() {propertyValues {
        AlphaCurve {keys []}
        BlueCurve {keys [<at 0 0.7865>,<at 1 1>]}
        GreenCurve {keys [<at 0 0.7865>,<at 1 0.3835>]}
        RedCurve {keys [<at 0 0.7865>,<at 1 0.2384>]}
       }}))
    ScaleA=true
    ScaleRGB=true
   )
   &SolveForcesAndVelocity(Clamp Velocity=false, Limit Acceleration=false)
  ]
  rendererClasses [&NiagaraMeshRendererProperties]
  renderers [
   {rendererClass=&NiagaraMeshRendererProperties, rendererIndex=0}
   {rendererClass=&NiagaraMeshRendererProperties, rendererIndex=1}
   {rendererClass=&NiagaraMeshRendererProperties, rendererIndex=2}
  ]
 }
 WispyTrail {
  emitterUpdateScript [
   &EmitterState(Life Cycle Mode=::NewEnumerator1, Loop Behavior=::NewEnumerator1, Loop Duration=3, UseLoopDelay=false)
   &SpawnParticlesFromOtherEmitter(
    Attribute Reader=DataInterface() {propertyValues {EmitterBinding {bindingMode=Other, emitterName=Bats}}}
    Spawn Rate=&RandomRangeFloat(Minimum=20, Maximum=40)
    Spawn Rate Per Particle Cap=10000000
    Use Maximum Age=false
    Use Minimum Age=false
   )
  ]
  particleSpawnScript [
   &InitializeParticle(UsePositionOffset=false, Write Material Random=true)
   &SampleParticlesFromOtherEmitter(
    Age Sampling:ENiagara_DirectReadSamplingMode=::NewEnumerator1
    Applied Velocity Scale:NiagaraFloat=&FloatFromCurve(
     FloatCurve=DataInterface() {propertyValues {Curve {keys [{interpMode=RCIM_Cubic, tangentMode=RCTM_User, time=0, value=0}, {interpMode=RCIM_Cubic, tangentMode=RCTM_User, time=0.4174, value=0.136}]}}}
     CurveIndex=$Emitter.Age
     Scale Curve=&RandomRangeFloat(Minimum=-1, Maximum=-3)
    )
    Lifetime Application:ENiagara_LifetimeInheritanceOptions=::NewEnumerator0
    Lifetime Sampling:ENiagara_DirectReadSamplingMode=::NewEnumerator0
    Mesh Scale Sampling:ENiagara_DirectReadSamplingMode=::NewEnumerator1
    Particle ID Sampling=::NewEnumerator1
    Ribbon UV Distance:ENiagara_DirectReadSamplingMode=::NewEnumerator0
    Sprite Size Sampling:ENiagara_DirectReadSamplingMode=::NewEnumerator1
    Velocity Sampling=::NewEnumerator0
   )
   &InitializeRibbon(
    Color=<rgba 0.366 1 0.2491 0.9>
    Lifetime=&FloatFromCurve(
     FloatCurve=DataInterface() {propertyValues {Curve {keys [
         {arriveTangent=-3.7746, interpMode=RCIM_Cubic, leaveTangent=-3.7746, tangentMode=RCTM_User, time=0, value=1.27}
         {arriveTangent=-0.2864, interpMode=RCIM_Cubic, leaveTangent=-0.2864, tangentMode=RCTM_User, time=0.353, value=0.6}
        ]}}}
     CurveIndex=$Emitter.Age
     Scale Curve=0.5
    )
    Position Mode:ENiagara_PositionInitializationMode=::NewEnumerator0
    Ribbon Width=1
    Write Material Random=true
   )
   /SetVariables(Particles.RibbonWidth=&Multiply_Float(
     A=&MakeFloatFromVector(Vector=$Output.SampleParticlesFromOtherEmitter.SampledScale, Channel=::NewEnumerator0)
     B=18
    )) {isSetParameters=true}
  ]
  particleUpdateScript [
   &self:ReadRibbonSourceParticleAge_06(PAR=DataInterface() {propertyValues {EmitterBinding {bindingMode=Other, emitterName=Bats}}}) {module=ReadRibbonSourceParticleAge_06}
   &ParticleState()
   &SolveForcesAndVelocity(Clamp Velocity=false, Limit Acceleration=false)
   &DynamicMaterialParameters(
    Index 0 Param 1=1
    Module.Index0Float0Write=true
    Module.Index0Float1Write=false
    Module.Index0Float2Write=true
    Module.Index0Float3Write=true
   ) {module=DynamicMaterialParameters001}
   &ScaleColor(
    Scale Alpha=&FloatFromCurve(
     FloatCurve=DataInterface() {propertyValues {Curve {keys [
         {tangentMode=RCTM_User, time=0, value=1}
         {arriveTangent=-0.0405, interpMode=RCIM_Cubic, leaveTangent=-0.0405, tangentMode=RCTM_User, time=0.0897, value=1}
         {arriveTangent=-2.5054, interpMode=RCIM_Cubic, leaveTangent=-2.5054, tangentMode=RCTM_User, time=0.65, value=0}
        ]}}}
     Scale Curve=0.75
    )
    Scale RGB=&MakeVectorFromLinearColorRGB(Color=&ColorFromCurve(ColorCurve=DataInterface() {propertyValues {
        AlphaCurve {keys [<at 0 1 RCIM_Cubic>,<at 1 0 RCIM_Cubic>]}
        BlueCurve {keys [<at 0 0 RCIM_Cubic>,<at 1 0.0219 RCIM_Cubic>]}
        GreenCurve {keys [<at 0 1 RCIM_Cubic>,<at 1 0.125 RCIM_Cubic>]}
        RedCurve {keys [<at 0 0.2348 RCIM_Cubic>,<at 1 0.0548 RCIM_Cubic>]}
       }}))
    ScaleA=true
    ScaleRGB=true
   )
   &ScaleRibbonWidth(Ribbon Width Scale=&FloatFromCurve(FloatCurve=DataInterface() {propertyValues {Curve {
        defaultValue=1
        keys [
         {tangentMode=RCTM_User, time=0, value=1}
         {arriveTangent=-0.0405, interpMode=RCIM_Cubic, leaveTangent=-0.0405, tangentMode=RCTM_User, time=0.7009, value=1}
         {arriveTangent=-8.0089, interpMode=RCIM_Cubic, leaveTangent=-8.0089, tangentMode=RCTM_User, time=1, value=0}
        ]
       }}}))
  ]
  rendererClasses [&NiagaraRibbonRendererProperties]
  renderers [{rendererClass=&NiagaraRibbonRendererProperties}]
 }
}
summary {
 systemName=NS_Emote_LanternStroll_Ghosts
 emitters [
  {emitterName=Bats, rendererClasses [&NiagaraMeshRendererProperties, &NiagaraSpriteRendererProperties]}
  {emitterName=WispyTrail, rendererClasses [&NiagaraRibbonRendererProperties]}
  {emitterName=Ghosts, rendererClasses [&NiagaraMeshRendererProperties]}
 ]
 userVariables [
  User.bIsFrontEnd {type=NiagaraBool}
  User.bIsFrontEndPreview {type=NiagaraBool}
 ]
}
