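# Niagara system for the Lantern Stroll emote that adds a flickering head glow and drifting green fog around the player.
# Skeletal-mesh sampling positions stacked glow planes and light renderers, driven by shared glow color and flicker.
# Two sprite fog emitters quicken with owner velocity and blend between local and world space.
# Time of day, front-end state and platform quality adjust intensity, while a head-hidden flag sets light visibility tags.
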NiagaraSystem /BRCosmetics/Animation/Game/MainPlayer/Emotes/LanternStroll/FX/NS_Emote_LanternStroll_HeadGlow
dependencies {
 usedDataInterfaces [&NiagaraDataInterfacePlatformSet, &NiagaraDataInterfaceCurve, &NiagaraDataInterfaceSkeletalMesh]
 usedDynamicInputs [&Waveform]
 usedModules [
  &self:SystemSpawnScript.SystemScriptSource.SystemScriptGraph.NiagaraNodeAssignment_0.SetVariables_#1
  &self:SystemSpawnScript.SystemScriptSource.SystemScriptGraph.NiagaraNodeAssignment_1.SetVariables_#2
  &SystemState
  &BlendSpaceHelper
  &self:SystemSpawnScript.SystemScriptSource.SystemScriptGraph.NiagaraNodeAssignment_3.SetVariables_#3
 ]
}
emitters {
 FOG {
  emitterUpdateScript [
   &EmitterState(
    Delay First Loop Only=true
    Inactive Response=::NewEnumerator1
    Life Cycle Mode=::NewEnumerator1
    Loop Delay=0.4
    Loop Duration=1
    UseLoopDelay=true
   )
   &SpawnRate(
    Spawn Probability=0.5
    SpawnRate=&FN_FrontEndFloatModifier(
     Gameplay=&MakeCustomFloatFromBool(
      Bool=&SetBoolByFloatComparison(A=&VectorLength(Vector=$User.OwnerLocalVelocity), B=90)
      True Float=30
      False Float=20
     )
     Front End=15
    )
    Use Spawn Probability=true
   )
  ]
  particleSpawnScript [
   &InitializeParticle(
    Color Maximum=<rgba 0.4939 1 0.4405 1>
    Color Minimum=<rgba 0.2789 0.6979 0.1292 0.7>
    Color Mode=::NewEnumerator2
    Lifetime=&RandomRangeFloat(Minimum=2, Maximum=2.5)
    Position Offset=<0 -20 0>
    Sprite Rotation Mode=::NewEnumerator2
    Sprite Size Mode=::NewEnumerator1
    Uniform Sprite Size Max=25
    Uniform Sprite Size Min=12
    UsePositionOffset=true
    Write Material Random=true
   )
   &AddVelocity(Velocity=&MakeVector(
     X=0
     Y=&FN_FrontEndFloatModifier(
      Gameplay=&MakeCustomFloatFromBool(
       Bool=&SetBoolByFloatComparison(A=&VectorLength(Vector=$User.OwnerLocalVelocity), B=90)
       True Float=&RandomRangeFloat(Minimum=-20, Maximum=-50)
       False Float=&RandomRangeFloat(Minimum=-10, Maximum=-15)
      )
      Front End=7
     )
     Z=0
    ))
   &ShapeLocation(MaskSpawnGroup=false, Non Uniform Scale=<1 1 0.5>, Override Seed=false, Sphere Radius=40, UseOwnerScale=true)
   &CameraOffset(Camera Offset Amount=15)
  ]
  particleUpdateScript [
   &ParticleState()
   &SubUVAnimation(
    End Frame=64
    SubUV Animation Mode:ENiagara_SubUVLookupMode=::NewEnumerator1
    SubUV Curve=DataInterface() {propertyValues {Curve {keys [<at 0 0 RCIM_Cubic>,<at 1 0.8 RCIM_Cubic>]}}}
   )
   &Drag(
    Drag=&Multiply_Float(
     A=&RandomRangeFloat(Minimum=0.5, Maximum=1)
     B=&FloatFromCurve(FloatCurve=DataInterface() {propertyValues {Curve {keys [
          {interpMode=RCIM_Cubic, leaveTangent=0.248, tangentMode=RCTM_Break, time=0, value=0}
          {interpMode=RCIM_Cubic, tangentMode=RCTM_Break, time=1, value=1}
         ]}}})
    )
    Use Linear Drag=true
    Use Rotational Drag=false
   )
   &SolveForcesAndVelocity(Clamp Velocity=false, Limit Acceleration=false)
   &ScaleColor(
    Scale Alpha=&FloatFromCurve(
     FloatCurve=DataInterface() {propertyValues {Curve {keys [
         {arriveTangent=3.5962, interpMode=RCIM_Cubic, leaveTangent=3.5962, tangentMode=RCTM_User, time=0, value=0}
         {
          arriveTangent=0.1817
          arriveTangentWeight=0.1842
          interpMode=RCIM_Cubic
          leaveTangent=0.1817
          leaveTangentWeight=0.7369
          tangentMode=RCTM_User
          time=0.5037
          value=0.7647
         }
         {arriveTangent=-4.7233, tangentMode=RCTM_User, time=1, value=0}
        ]}}}
     Scale Curve=&RandomRangeFloat(
      Minimum=0.02
      Maximum=&FNTimeOfDayScalar(Night Scale=0.03, Morning Scale=0.03, Day Scale=0.05, Evening Scale=0.03)
     )
    )
    ScaleA=true
    ScaleRGB=true
   )
   &ScaleSpriteSize(
    Uniform Curve Scale=10
    Uniform Curve Sprite Scale=DataInterface() {propertyValues {Curve {defaultValue=1, keys [{tangentMode=RCTM_User, time=0, value=0.7}, {tangentMode=RCTM_User, time=1, value=1}]}}}
   )
   &DynamicMaterialParameters(
    Index 0 Param 2=1
    Index 0 Param 3=&RandomRangeFloat(Minimum=0.1, Maximum=0.45)
    Index 0 Param 4=1
    Module.Index0Param0Write=true
    Module.Index0Param1Write=true
    Module.Index0Param2Write=true
    Module.Index0Param3Write=true
   )
   &BlendSpace("Position Blend (World <-> Local)":NiagaraFloat=&MakeCustomFloatFromBool(
     Bool=&SetBoolByFloatComparison(A=&VectorLength(Vector=$Engine.Owner.Velocity), B=400)
     True Float=1
     False Float=&FloatFromCurve(FloatCurve=DataInterface() {propertyValues {Curve {keys [<at 0 1 RCIM_Cubic>, {arriveTangent=-3.5186, interpMode=RCIM_Cubic, leaveTangent=-3.5186, tangentMode=RCTM_User, time=0.8, value=0}]}}})
     Translation Contribution:NiagaraFloat=1
     Rotation Contribution:NiagaraFloat=1
    ))
   &BlendSpace(Rotation Contribution=0.5, Translation Contribution=0.95) {enabled=false, module=BlendSpace001}
  ]
  rendererClasses [&NiagaraSpriteRendererProperties]
  renderers [{rendererClass=&NiagaraSpriteRendererProperties}]
 }
 Ground_fog001 {
  emitterUpdateScript [
   &EmitterState(Inactive Response=::NewEnumerator1, Life Cycle Mode=::NewEnumerator1, Loop Duration=1, UseLoopDelay=false)
   &SpawnRate(
    Spawn Probability=0.5
    SpawnRate=&FN_FrontEndFloatModifier(
     Gameplay=&MakeCustomFloatFromBool(
      Bool=&SetBoolByFloatComparison(A=&VectorLength(Vector=$User.OwnerLocalVelocity), B=90)
      True Float=30
      False Float=20
     )
     Front End=20
    )
    Use Spawn Probability=true
   )
  ]
  particleSpawnScript [
   &InitializeParticle(
    Color Maximum=<rgba 0.4939 1 0.4405 0.45>
    Color Minimum=<rgba 0.2789 0.6979 0.1292 0.125>
    Color Mode=::NewEnumerator2
    Lifetime=&RandomRangeFloat(Minimum=3.65, Maximum=4.7)
    Position Offset=<0 25 0>
    Sprite Size Max=<50 35>
    Sprite Size Min=<20 10>
    Sprite Size Mode=::NewEnumerator2
    UsePositionOffset=true
    Write Material Random=true
   )
   &ShapeLocation(
    MaskSpawnGroup=false
    Non Uniform Scale=<1 1 0.1>
    Offset=<0 35 -50>
    Offset Mode=::NewEnumerator0
    Override Seed=false
    Sphere Radius=125
    UseOwnerScale=true
   ) {module=ShapeLocation001}
   &AddVelocity(Velocity=&MakeVector(
     X=0
     Y=&FN_FrontEndFloatModifier(
      Gameplay=&MakeCustomFloatFromBool(
       Bool=&SetBoolByFloatComparison(A=&VectorLength(Vector=$User.OwnerLocalVelocity), B=90)
       True Float=-75
       False Float=-45
      )
      Front End=-35
     )
     Z=0
    ))
   &ShapeLocation(
    Box Size=<100 200 0>
    MaskSpawnGroup=false
    Non Uniform Scale=<1 1 0.1>
    Offset=<0 -40 -55>
    Offset Mode=::NewEnumerator0
    Override Seed=false
    Rotation Coordinate Space=::Simulation
    Shape Primitive=::NewEnumerator2
    Surface Only=false
    UseOwnerScale=true
   ) {enabled=false}
   &CameraOffset(Camera Offset Amount=15)
  ]
  particleUpdateScript [
   &ParticleState()
   &SubUVAnimation(
    End Frame=64
    SubUV Animation Mode:ENiagara_SubUVLookupMode=::NewEnumerator1
    SubUV Curve=DataInterface() {propertyValues {Curve {keys [<at 0 0 RCIM_Cubic>,<at 1 0.4 RCIM_Cubic>]}}}
   )
   &Drag(
    Drag=&Multiply_Float(
     A=&RandomRangeFloat(Minimum=0.8, Maximum=1.2)
     B=&FloatFromCurve(FloatCurve=DataInterface() {propertyValues {Curve {keys [
          {interpMode=RCIM_Cubic, leaveTangent=0.248, tangentMode=RCTM_Break, time=0, value=0}
          {interpMode=RCIM_Cubic, tangentMode=RCTM_Break, time=1, value=1}
         ]}}})
    )
    Use Linear Drag=true
    Use Rotational Drag=false
   )
   &ScaleColor(
    Scale Alpha=&FloatFromCurve(
     FloatCurve=DataInterface() {propertyValues {Curve {keys [
         {arriveTangent=5, interpMode=RCIM_Cubic, time=0, value=0}
         {arriveTangent=1.5385, interpMode=RCIM_Cubic, leaveTangent=1.5385, time=0.2, value=0.7}
         {
          arriveTangent=0.1817
          arriveTangentWeight=0.1842
          interpMode=RCIM_Cubic
          leaveTangent=0.1817
          leaveTangentWeight=0.7369
          tangentMode=RCTM_User
          time=0.65
          value=1
         }
         <at 1 0>
        ]}}}
     Scale Curve=&RandomRangeFloat(
      Minimum=&FNTimeOfDayScalar(Night Scale=0.035, Morning Scale=0.07, Day Scale=0.05, Evening Scale=0.07)
      Maximum=&FNTimeOfDayScalar(Night Scale=0.05, Morning Scale=0.07, Day Scale=0.07, Evening Scale=0.1)
     )
    )
    ScaleA=true
    ScaleRGB=true
   )
   &ScaleSpriteSize(
    Uniform Curve Scale=10
    Uniform Curve Sprite Scale=DataInterface() {propertyValues {Curve {defaultValue=1, keys [{tangentMode=RCTM_User, time=0, value=0.7}, {tangentMode=RCTM_User, time=1, value=1}]}}}
   )
   &DynamicMaterialParameters(
    Index 0 Param 2=1
    Index 0 Param 3=&RandomRangeFloat(Minimum=0.05, Maximum=0.2)
    Index 0 Param 4=1
    Module.Index0Param0Write=true
    Module.Index0Param1Write=true
    Module.Index0Param2Write=true
    Module.Index0Param3Write=true
   )
   &BlendSpace("Position Blend (World <-> Local)":NiagaraFloat=&MakeCustomFloatFromBool(
     Bool=&SetBoolByFloatComparison(A=&VectorLength(Vector=$Engine.Owner.Velocity), B=400)
     True Float=1
     False Float=&FloatFromCurve(FloatCurve=DataInterface() {propertyValues {Curve {keys [<at 0 1 RCIM_Cubic>, {arriveTangent=-3.5186, interpMode=RCIM_Cubic, leaveTangent=-3.5186, tangentMode=RCTM_User, time=0.5, value=0}]}}})
     Translation Contribution:NiagaraFloat=1
     Rotation Contribution:NiagaraFloat=1
    ))
   &SolveForcesAndVelocity(Clamp Velocity=false, Limit Acceleration=false)
  ]
  rendererClasses [&NiagaraSpriteRendererProperties]
  renderers [{rendererClass=&NiagaraSpriteRendererProperties}]
 }
 Lights {
  emitterUpdateScript [
   &EmitterState(Life Cycle Mode=::NewEnumerator1, Loop Behavior=::NewEnumerator1, Loop Duration Mode=::NewEnumerator1, UseLoopDelay=false)
   &SpawnBurst_Instantaneous(Use Spawn Probability=false, UseLoopCountLimit=false)
   &SpawnBurst_Instantaneous(Spawn Time=3.5986, Use Spawn Probability=false, UseLoopCountLimit=false) {module=SpawnBurst_Instantaneous001}
   &SpawnBurst_Instantaneous(Spawn Time=1.3417, Use Spawn Probability=false, UseLoopCountLimit=false) {module=SpawnBurst_Instantaneous002}
  ]
  particleSpawnScript [&InitializeParticle(
    Color=&MultiplyLinearColor(Linear Color=$System.Glow Color, Scale Factor=&MakeVector4_FromFloat(Float=1))
    Lifetime=5
    Sprite Size Mode=::NewEnumerator3
    Uniform Sprite Size=9
    UsePositionOffset=false
    Write Material Random=true
   )]
  particleUpdateScript [
   &ParticleState(
    Kill Particles When Lifetime Has Elapsed=false
    Let Infinitely Lived Particles Die When Emitter Deactivates=true
    Loop Particles Lifetime=true
   )
   &SkeletalMeshLocation(Bone Sampling Mode=::NewEnumerator4, MaskSpawnGroup=false, Sampled Position Offset=<-2.5 -10 -17.5>, Skeletal Mesh=$System.Skeletal Mesh)
   &ScaleColor(
    Scale Alpha=&Multiply_Float(
     A=&FloatFromCurve(FloatCurve=DataInterface() {propertyValues {Curve {keys [<at 0 0 RCIM_Cubic>,<at 1 1 RCIM_Cubic>]}}}, CurveIndex=$Emitter.Age)
     B=&Multiply_Float(
      A=$System.Flicker
      B=&Multiply_Float(
       A=&FNTimeOfDayScalar(Night Scale=0.8, Morning Scale=0.8, Day Scale=0.5, Evening Scale=0.8)
       B=33
      )
     )
    )
    ScaleA=true
    ScaleRGB=true
   )
   /SetVariables(
    Particles.PointLightColor:LinearColor=&MakeLinearColorFromVectorAndFloat(
     "Vector (RGB)"=&MakeVectorFromLinearColorRGB(Color=$System.Glow Color)
     "Float (Alpha)"=$System.Flicker
    )
    Particles.SpotLightIntensity:NiagaraFloat=&Multiply_Float(
     A=&FloatFromCurve(FloatCurve=DataInterface() {propertyValues {Curve {keys [<at 0 0 RCIM_Cubic>,<at 1 1 RCIM_Cubic>]}}}, CurveIndex=$Emitter.Age)
     B=&Multiply_Float(
      A=$System.Flicker
      B=&FN_FrontEndFloatModifier(
       Gameplay=&Multiply_Float(
        A=&FNTimeOfDayScalar(Night Scale=0.5, Morning Scale=0.5, Day Scale=0.8, Evening Scale=0.5)
        B=350
       )
       Front End=0
      )
     )
    )
    Particles.SpotLightPosition:NiagaraPosition=&AddVectorToPosition(Position=$Particles.SkeletalMeshLocation.SampledPosition, Vector=<0 -15 -75>)
   ) {isSetParameters=true}
   /SetVariables(Particles.VisibilityTag=&MakeCustomIntFromBool(Bool=$User.bisHeadHidden, True INT=1, False INT=0)) {isSetParameters=true, module=SetVariables_#4}
  ]
  rendererClasses [&NiagaraLightRendererProperties, &NiagaraComponentRendererProperties]
  renderers [
   {rendererClass=&NiagaraLightRendererProperties, rendererIndex=0}
   {rendererClass=&NiagaraComponentRendererProperties, rendererIndex=1}
  ]
 }
 Stacked_Plane_Glow004 {
  emitterUpdateScript [
   &EmitterState(Life Cycle Mode=::NewEnumerator1, Loop Behavior=::NewEnumerator1, Loop Duration Mode=::NewEnumerator1, UseLoopDelay=false)
   &SpawnBurst_Instantaneous(Spawn Count=6, Use Spawn Probability=false, UseLoopCountLimit=false)
  ]
  particleSpawnScript [
   &FN_SetSpawnGroup(Number of Spawn Groups=30)
   &InitializeParticle(
    Color=$System.Glow Color
    Lifetime=8.8
    Mesh Scale=&Multiply_VectorByFloat(
     Vector=&Multiply_Vector(
      A=&MakeVector(
       X=0.95
       Y=&MakeCustomFloatFromBool(
        Bool=&IsPlatformSetActive(New Platform Set=DataInterface() {propertyValues {Platforms {qualityLevelMask=-31}}})
        True Float=1.2
        False Float=1.7
       )
       Z=1.1
      )
      B=&MakeVector(
       X=&Add_Float(A=1, B=&Multiply_FloatByInt(Float=0.1, Integer=&SpawnGroup()))
       Y=&Add_Float(A=1, B=&Multiply_FloatByInt(Float=0.06, Integer=&SpawnGroup()))
       Z=&Add_Float(A=1, B=&Multiply_FloatByInt(Float=0.1, Integer=&SpawnGroup()))
      )
     )
     Float=0.9
    )
    Mesh Scale Mode=::NewEnumerator4
    UsePositionOffset=false
    Write Material Random=true
   ) {module=InitializeParticle001}
   &CameraOffset()
   &DynamicMaterialParameters(
    Index 0 Param 1=&Multiply_FloatByInt(Float=0.1, Integer=&SpawnGroup())
    Module.Index0Param0Write=true
    Module.Index0Param1Write=false
    Module.Index0Param2Write=false
    Module.Index0Param3Write=false
   ) {module=DynamicMaterialParameters001}
   &InitialMeshOrientation(Mesh Axis To Orient=<0 1 0>, Mesh Orientation Mode=::NewEnumerator2, Orientation Vector=<0 1 0>, Use Rotation Vector=true) {module=InitialMeshOrientation001}
  ]
  particleUpdateScript [
   &ParticleState(
    Kill Particles When Lifetime Has Elapsed=false
    Let Infinitely Lived Particles Die When Emitter Deactivates=true
    Loop Particles Lifetime=true
   )
   &SkeletalMeshLocation(Bone Sampling Mode=::NewEnumerator4, MaskSpawnGroup=false, Sampled Position Offset=<0 -5 -15>, Skeletal Mesh=$System.Skeletal Mesh) {module=SkeletalMeshLocation001}
   &ScaleColor(
    Scale Alpha=&Multiply_Float(
     A=&MakeCustomFloatFromBool(
      Bool=&IsPlatformSetActive(New Platform Set=DataInterface() {propertyValues {Platforms {qualityLevelMask=-31}}})
      True Float=0.4
      False Float=1
     )
     B=&Multiply_Float(
      A=&FloatFromCurve(FloatCurve=DataInterface() {propertyValues {Curve {keys [<at 0 0 RCIM_Cubic>,<at 1 1 RCIM_Cubic>]}}}, CurveIndex=$Emitter.Age)
      B=&Multiply_Float(
       A=&Multiply_Float(
        A=$System.Flicker
        B=&FloatFromCurve(
         FloatCurve=DataInterface() {propertyValues {Curve {keys [
             {interpMode=RCIM_Cubic, tangentMode=RCTM_User, time=0, value=0.8}
             {interpMode=RCIM_Cubic, tangentMode=RCTM_User, time=0.5, value=1.3}
             {tangentMode=RCTM_User, time=1, value=0.8}
            ]}}}
         CurveIndex=Linked() {linkedVariable {name=NPC.FortniteNPC.TimeOfDay_0-1, type=NiagaraFloat}}
        )
       )
       B=&Power(a=0.875, b=&MakeFloatFromInt(Int=&SpawnGroup()))
      )
     )
    )
    Scale RGB=&VectorFromFloat(Value=&FloatFromCurve(
      FloatCurve=DataInterface() {propertyValues {Curve {keys [
          {interpMode=RCIM_Cubic, tangentMode=RCTM_User, time=0, value=1}
          {interpMode=RCIM_Cubic, tangentMode=RCTM_User, time=0.5, value=1}
          {tangentMode=RCTM_User, time=1, value=1}
         ]}}}
      CurveIndex=Linked() {linkedVariable {name=NPC.FortniteNPC.TimeOfDay_0-1, type=NiagaraFloat}}
      Scale Curve=&MakeCustomFloatFromBool(
       Bool=&IsPlatformSetActive(New Platform Set=DataInterface() {propertyValues {Platforms {qualityLevelMask=-31}}})
       True Float=1.75
       False Float=1
      )
     ))
    ScaleA=true
    ScaleRGB=true
   ) {module=ScaleColor001}
   /SetVariables(Particles.VisibilityTag=&MakeCustomIntFromBool(Bool=$User.bisHeadHidden, True INT=1, False INT=0)) {enabled=false, isSetParameters=true, module=SetVariables_#4}
  ]
  rendererClasses [&NiagaraMeshRendererProperties]
  renderers [{rendererClass=&NiagaraMeshRendererProperties, rendererIndex=0}]
 }
}
summary {
 systemName=NS_Emote_LanternStroll_HeadGlow
 emitters [
  {emitterName=Stacked_Plane_Glow004, rendererClasses [&NiagaraMeshRendererProperties]}
  {emitterName=Lights, rendererClasses [&NiagaraLightRendererProperties, &NiagaraComponentRendererProperties]}
  {emitterName=Ground_fog001, rendererClasses [&NiagaraSpriteRendererProperties]}
  {emitterName=FOG, rendererClasses [&NiagaraSpriteRendererProperties]}
 ]
 userVariables [
  User.bIsFrontEnd {type=NiagaraBool}
  User.bIsFrontEndPreview {type=NiagaraBool}
  User.bisHeadHidden {type=NiagaraBool}
  User.OwnerLocalVelocity {type=Vector3f, description="", defaultValue {struct=Vector3f, value=<0 0 0>}}
 ]
}
