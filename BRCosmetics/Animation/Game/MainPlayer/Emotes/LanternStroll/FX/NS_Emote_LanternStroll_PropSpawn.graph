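# Niagara effect for the LanternStroll emote's prop spawn, puffing smoke and embers.
# Smoke pairs a burst with a tapering spawn rate, drawing flipbook sprites tinted green that grow and fade.
# Embers burst in two spawn groups, swirl with vortex and curl noise, stretch with speed, cool in color, and scale by camera distance and time of day.
# AirglowLWE and Spiral emitters have no renderers.
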NiagaraSystem /BRCosmetics/Animation/Game/MainPlayer/Emotes/LanternStroll/FX/NS_Emote_LanternStroll_PropSpawn
dependencies {
 usedDataInterfaces [
  &NiagaraDataInterfaceCurve
  &NiagaraDataInterfaceSpriteRendererInfo
  &NiagaraDataInterfaceColorCurve
 ]
 usedDynamicInputs [&FN_LOD_Distance_Remap]
 usedModules [
  &SystemState
  &self:SystemSpawnScript.SystemScriptSource.SystemScriptGraph.NiagaraNodeAssignment_0.SetVariables_#1
 ]
}
emitters {
 AirglowLWE {rendererClasses [], renderers []}
 Embers {
  emitterUpdateScript [
   &EmitterState(
    Inactive Response=::NewEnumerator1
    Life Cycle Mode=::NewEnumerator1
    Loop Behavior=::NewEnumerator1
    Loop Delay=0.1
    Loop Duration=1
    UseLoopDelay=true
   )
   &SpawnBurst_Instantaneous(Spawn Count=15, Use Spawn Probability=false, UseLoopCountLimit=false)
   &SpawnBurst_Instantaneous(Spawn Count=5, Spawn Group=1, Use Spawn Probability=false, UseLoopCountLimit=false) {module=SpawnBurst_Instantaneous001}
  ]
  particleSpawnScript [
   /SetVariables(Particles.VisibilityTag=&SpawnGroup()) {isSetParameters=true}
   &InitializeParticle(
    Color=<rgba 1 1 1 2>
    Lifetime=&UniformRangedFloat(Minimum=1.1, Maximum=0.9)
    Mass Mode=::NewEnumerator1
    Sprite Size Mode=::NewEnumerator1
    Uniform Sprite Size Max=1.5
    Uniform Sprite Size Min=0.8
    UsePositionOffset=false
    Write Material Random=true
   ) {module=InitializeParticle001}
   &ShapeLocation(
    MaskSpawnGroup=false
    Non Uniform Scale=<1 1 1.5>
    Offset Coordinate Space=::World
    Offset Mode=::NewEnumerator0
    Override Seed=false
    Sphere Radius=18
    UseOwnerScale=true
   )
   &AddVelocity(Velocity=&RandomRangeVector(Minimum=<0 0 -25>, Maximum=<0 0 0>)) {module=AddVelocity001}
   &CameraOffset(Camera Offset Amount=-10)
  ]
  particleUpdateScript [
   &ParticleState()
   &ApplyOwnerScaleToAttributes(Scale Initial Velocity=true)
   &ScaleSpriteSize(
    Scale Sprite Size Mode=::NewEnumerator0
    Uniform Scale Factor=&Lerp_Float(A=1.5, B=2, Alpha=$System.DistanceScalar)
   )
   &Drag(
    Drag=&Multiply_Float(
     A=&FloatFromCurve(
      FloatCurve=DataInterface() {propertyValues {Curve {keys [<at 0 0.5>, {arriveTangent=0.5, interpMode=RCIM_Cubic, leaveTangent=0.5, time=0.3, value=0.5}, <at 1 1 RCIM_Cubic>]}}}
      Scale Curve=12
     )
     B=1
    )
    Use Linear Drag=true
    Use Rotational Drag=false
   )
   &VortexForce(Origin Pull Amount=500, UseOriginOffset=true, Vortex Axis=<0 0 -1>, Vortex Force Amount=-600)
   /Niagara/Modules/Update/Forces/CurlNoiseForce(
    Noise Frequency=10
    Noise Strength=&FloatFromCurve(FloatCurve=DataInterface() {propertyValues {Curve {keys [<at 0 1>,<at 1 0>]}}}, Scale Curve=2000)
    Pan Noise=true
    Pan Noise Field=<0.1 0.1 0.5>
   )
   &GravityForce(Gravity=&RandomRangeVector(Minimum=<0 0 900>, Maximum=<0 0 -500>))
   &InheritSourceMovement(
    Applied Position Scale=&MakeVector(
     X=0
     Y=0
     Z=&FloatFromCurve(
      FloatCurve=DataInterface() {propertyValues {Curve {
         defaultValue=1
         keys [
          {interpMode=RCIM_Cubic, tangentMode=RCTM_User, time=0, value=0.0874}
          {arriveTangent=3.8536, interpMode=RCIM_Cubic, leaveTangent=3.8536, tangentMode=RCTM_User, time=0.1, value=0.5931}
         ]
        }}}
      Scale Curve=-1
     )
    )
    ApplyForce=false
    ApplyPosition=true
    ApplyPositionFalloff=false
   )
   &SolveForcesAndVelocity(Clamp Velocity=false, Limit Acceleration=false)
   &ScaleSpriteSizeBySpeed(
    Max Scale Factor=<0.8 5>
    Min Scale Factor=<0.8 0.8>
    Sample Scale Factor By Curve=true
    Scale Factor Curve=DataInterface() {propertyValues {Curve {keys [
        <at -0.0077 0.0765 RCIM_Cubic>
        {arriveTangent=1.2463, interpMode=RCIM_Cubic, leaveTangent=1.2463, tangentMode=RCTM_User, time=0.9985, value=0.9549}
       ]}}}
    Velocity Threshold=300
   )
   &ScaleColor(
    Scale Alpha=&FloatFromCurve(
     FloatCurve=DataInterface() {propertyValues {Curve {keys [<at 0.25 1>,<at 1 0>]}}}
     Scale Curve=&MaskFloatBySpawnGroup(
      Masked Float=&FNTimeOfDayScalar(Night Scale=0.1, Morning Scale=0.2, Day Scale=0.3, Evening Scale=0.2)
      Passthrough Float=&FNTimeOfDayScalar(Night Scale=0.1, Morning Scale=0.2, Evening Scale=0.2)
     )
    )
    Scale RGB=&MakeVectorFromLinearColorRGB(Color=&ColorFromCurve(ColorCurve=DataInterface() {propertyValues {
        AlphaCurve {keys [<at 0 1 RCIM_Cubic>,<at 0.7107 1>]}
        BlueCurve {keys [<at 0 14.6701 RCIM_Cubic>,<at 0.368 3.776>,<at 0.855 0.1753 RCIM_Cubic>]}
        GreenCurve {keys [<at 0 36.7481 RCIM_Cubic>,<at 0.368 13.2325>,<at 0.855 0.3243 RCIM_Cubic>]}
        RedCurve {keys [<at 0 50 RCIM_Cubic>,<at 0.368 25>,<at 0.855 1 RCIM_Cubic>]}
       }}))
    ScaleA=true
    ScaleRGB=true
   )
  ]
  rendererClasses [&NiagaraSpriteRendererProperties]
  renderers [
   {rendererClass=&NiagaraSpriteRendererProperties}
   {rendererClass=&NiagaraSpriteRendererProperties, rendererIndex=1}
  ]
 }
 Smoke {
  emitterUpdateScript [
   &EmitterState(Life Cycle Mode=::NewEnumerator1, Loop Behavior=::NewEnumerator1, Loop Duration=1.5, UseLoopDelay=false)
   &SpawnRate(
    SpawnRate=&FloatFromCurve(
     FloatCurve=DataInterface() {propertyValues {Curve {keys [<at 0 1>,<at 0.1475 1>,<at 0.3475 0>]}}}
     CurveIndex=$Emitter.NormalizedLoopAge
     Scale Curve=6
    )
    Use Spawn Probability=false
   )
   &SpawnBurst_Instantaneous(Spawn Count=3, Use Spawn Probability=false, UseLoopCountLimit=false)
  ]
  particleSpawnScript [
   &InitializeParticle(
    Lifetime Max=1.4
    Lifetime Min=0.7
    Lifetime Mode=::NewEnumerator1
    Position Offset=&MakeVector(
     X=0
     Y=&FloatFromCurve(FloatCurve=DataInterface() {propertyValues {Curve {keys [<at 0 0>,<at 0.7 25>]}}}, CurveIndex=$Emitter.NormalizedLoopAge)
     Z=0
    )
    Sprite Rotation Mode=::NewEnumerator2
    Sprite Size Mode=::NewEnumerator1
    Uniform Sprite Size Max=42
    Uniform Sprite Size Min=32
    UsePositionOffset=true
    Write Material Random=true
   )
   &ShapeLocation(
    MaskSpawnGroup=false
    Non Uniform Scale=<0.2 0.2 7>
    Offset=<0 0 -5>
    Offset Mode=::NewEnumerator0
    Override Seed=false
    Sphere Radius=5
    UseOwnerScale=true
   )
   &AddVelocity(
    Cone Axis=&MakeVector(X=0, Y=0, Z=-1)
    Module.UseConeFalloff=false
    Module.UseVelDistribution=false
    Override Seed=false
    Velocity Mode=::NewEnumerator2
    Velocity Speed=&RandomRangeFloat(Minimum=40, Maximum=20)
   )
   &AddVelocity(
    Module.Use Falloff Distance=false
    Override Seed=false
    Velocity Mode=::NewEnumerator1
    Velocity Origin=$Particles.ShapeLocation.ShapePosition
    Velocity Speed=&RandomRangeFloat(Minimum=0, Maximum=25)
   ) {module=AddVelocity001}
   &CameraOffset(Camera Offset Amount=15)
  ]
  particleUpdateScript [
   &ParticleState()
   &SubUVAnimation(
    "Sprite Renderer "=DataInterface() {propertyValues {SpriteRenderer=&self:Smoke_1.NiagaraSpriteRendererProperties_0}}
    UseEndFrame=false
    UseStartFrame=false
   )
   &ScaleSpriteSize(
    Uniform Curve Scale=1.45
    Uniform Curve Sprite Scale=DataInterface() {propertyValues {Curve {
       defaultValue=0
       keys [
        {arriveTangent=4.1785, interpMode=RCIM_Cubic, leaveTangent=4.1785, tangentMode=RCTM_User, time=0, value=0}
        {arriveTangent=1.1179, interpMode=RCIM_Cubic, leaveTangent=1.1179, tangentMode=RCTM_User, time=0.2001, value=0.7483}
        {arriveTangent=0.0224, interpMode=RCIM_Cubic, leaveTangent=0.0224, tangentMode=RCTM_User, time=1, value=1}
       ]
      }}}
   )
   &ScaleColor(
    Scale Alpha=&FloatFromCurve(
     FloatCurve=DataInterface() {propertyValues {Curve {
        defaultValue=0
        keys [
         {arriveTangent=9.0725, interpMode=RCIM_Cubic, leaveTangent=9.0725, tangentMode=RCTM_User, time=0, value=0.7366}
         {arriveTangent=-2.5249, interpMode=RCIM_Cubic, leaveTangent=-2.5249, tangentMode=RCTM_User, time=0.2536, value=1.2737}
         {arriveTangent=-0.0405, interpMode=RCIM_Cubic, leaveTangent=-0.0405, tangentMode=RCTM_User, time=1, value=0}
        ]
       }}}
     Scale Curve=0.25
    )
    Scale RGB=&MakeVectorFromLinearColorRGB(Color=&ColorFromCurve(ColorCurve=DataInterface() {propertyValues {
        AlphaCurve {keys []}
        BlueCurve {keys [<at 0 0.0505>,<at 1 0 RCIM_Cubic>]}
        GreenCurve {keys [<at 0 1>, {arriveTangent=-1.2172, interpMode=RCIM_Cubic, leaveTangent=-1.2172, tangentMode=RCTM_User, time=1, value=1}]}
        RedCurve {keys [<at 0 0.4633>, {arriveTangent=-1.6259, interpMode=RCIM_Cubic, leaveTangent=-1.6259, tangentMode=RCTM_User, time=1, value=0.2261}]}
       }}))
    ScaleA=true
    ScaleRGB=true
   )
   &Drag(Drag=&RandomRangeFloat(Minimum=0.8, Maximum=1.2), Use Linear Drag=true, Use Rotational Drag=true)
   &InheritSourceMovement(
    Applied Position Scale=&MakeVector(
     X=0
     Y=0
     Z=&FloatFromCurve(
      FloatCurve=DataInterface() {propertyValues {Curve {
         defaultValue=1
         keys [
          {interpMode=RCIM_Cubic, tangentMode=RCTM_User, time=0, value=0.0874}
          {arriveTangent=3.8536, interpMode=RCIM_Cubic, leaveTangent=3.8536, tangentMode=RCTM_User, time=0.0539, value=0.5931}
          {arriveTangent=1.0538, interpMode=RCIM_Cubic, leaveTangent=1.0538, time=0.5978, value=0.59}
          {interpMode=RCIM_Cubic, leaveTangent=0.547, time=0.6298, value=1.2}
         ]
        }}}
      Scale Curve=-1
     )
    )
    ApplyForce=false
    ApplyPosition=true
    ApplyPositionFalloff=false
   )
   &SolveForcesAndVelocity(Clamp Velocity=false, Limit Acceleration=false)
   &DynamicMaterialParameters(
    Index 0 Param 1=3
    Index 0 Param 2=1
    Index 0 Param 3=1
    Module.Index0Float0Write=true
    Module.Index0Float1Write=true
    Module.Index0Float2Write=true
    Module.Index0Float3Write=true
   )
  ]
  rendererClasses [&NiagaraSpriteRendererProperties]
  renderers [{rendererClass=&NiagaraSpriteRendererProperties}]
 }
 Spiral {rendererClasses [], renderers []}
}
summary {
 systemName=NS_Emote_LanternStroll_PropSpawn
 emitters [
  {emitterName=AirglowLWE, rendererClasses []}
  {emitterName=Smoke, rendererClasses [&NiagaraSpriteRendererProperties]}
  {emitterName=Embers, rendererClasses [&NiagaraSpriteRendererProperties]}
  {emitterName=Spiral, rendererClasses []}
 ]
 userVariables [
  User.bIsFrontEnd {type=NiagaraBool}
  User.bIsFrontEndPreview {type=NiagaraBool}
 ]
}
